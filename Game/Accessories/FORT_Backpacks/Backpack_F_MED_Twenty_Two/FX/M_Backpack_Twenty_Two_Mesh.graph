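# Translucent, unlit, two-sided material for Niagara mesh particles in the Twenty Two backpack effect, with an orange HDR color multiplied by particle color.
# World position offset combines SplineThicken width with a local-axis push driven by panning inky-smoke noise, each stage switchable by parameter.
# Opacity adds a sine-based band along the middle of U to panning noise raised by that band's inverse, then scales the optional mask by particle alpha.
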Material /Game/Accessories/FORT_Backpacks/Backpack_F_MED_Twenty_Two/FX/M_Backpack_Twenty_Two_Mesh
graph {
 properties {
  bUsedWithNiagaraMeshParticles=true
  blendMode=BLEND_Translucent
  pixelDepthOffsetMode=PDOM_Legacy
  shadingModel=MSM_Unlit
  twoSided=true
  usedShadingModels=Unlit
 }
 nodes [
  n1 = ScalarParameter() {defaultValue=1.5, group=WPO, parameterName=SplineThicken_Width}
  n2 = Constant()
  n3 = Constant3Vector() {constant=<rgba 1 0 0 0>}
  n4 = TextureCoordinate() {uTiling=0.5, vTiling=0.2}
  n5 = LinearInterpolate(Alpha=TextureSample(UVs=Panner(Coordinate=@n4) {speedY=1}) {texture=&T_Inky_Smoke_Tile}.R)
  n6 = Multiply(A=Transform(None=Multiply(A=@n3, B=@n5)) {transformSourceType=TRANSFORMSOURCE_Local}) {constB=3}
  n7 = ScalarParameter() {defaultValue=1, group=WPO, parameterName=Use_WPO_Offset}
  LinearInterpolate(
   A=Constant()
   B=Add(
    A=MaterialFunctionCall(WidthBase=@n1, WidthTip=@n1) {materialFunction=&SplineThicken}.WorldPositionOffset
    B=LinearInterpolate(A=@n2, B=@n6, Alpha=@n7)
   )
   Alpha=ScalarParameter() {defaultValue=1, group=WPO, parameterName=Use_WPO}
  )
  n8 = ParticleColor()
  MaterialFunctionCall(Emissive=Multiply(A=@n8.RGB, B=VectorParameter() {group=Emissive, parameterName=color, defaultValue=<rgba 10 3.1333 0 1>}.RGB)) {materialFunction=&MF_FX_EmissiveAdjustments}
  n9 = ComponentMask(None=TextureCoordinate()) {r=true}
  n10 = Multiply(
   A=Subtract(A=Sine(None=@n9) {period=2}, B=ScalarParameter() {defaultValue=0.95, group=Alpha, parameterName=Middle Range}) {constB=0.95}
   B=ScalarParameter() {defaultValue=10, group=Alpha, parameterName=Edge_bright}
  ) {constB=10}
  n11 = ScalarParameter() {defaultValue=-0.3, group=Alpha, parameterName=Noise_LerpMin}
  n12 = ScalarParameter() {defaultValue=2, group=Alpha, parameterName=Noise_LerpMax}
  n13 = TextureSampleParameter2D(UVs=Panner(Coordinate=TextureCoordinate() {vTiling=1.5}) {speedX=1, speedY=0.2}) {group=Alpha, parameterName=Noise, samplerType=SAMPLERTYPE_Masks, texture=/Game/Weapons/WeaponSkins/Wraps/Textures/Mash/Mash_Noise}
  n14 = Add(A=Saturate(None=@n10), B=Power(Base=LinearInterpolate(A=@n11, B=@n12, Alpha=@n13.R) {constA=-0.3, constB=2}, Exp=OneMinus(None=@n10)))
  n15 = ScalarParameter() {defaultValue=1, group=Alpha, parameterName=Edge_Power}
  Multiply(
   A=@n8.A
   B=LinearInterpolate(
    A=Constant() {r=1}
    B=Power(Base=@n14, Exp=@n15) {constExponent=1}
    Alpha=ScalarParameter() {defaultValue=1, group=Alpha, parameterName=Use_Mask_For_Alpha}
   )
  )
 ]
}
